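# Niagara system for the CoreStreet glider's exhaust: four-nozzle jets, cones, glows, smoke, lights and GPU embers.
# System timeline ramps glows, lights and cone color, and limits embers to a startup burst.
# User.ForwardDot scales jet, smoke and ember spawning, with front-end overrides, while muzzle-flash spawning increases as it drops.
# Smoke is pushed along local X, bending against owner velocity by ForwardAndRightDot.
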NiagaraSystem /BRCosmetics/Gadgets/Assets/VinderTech_GliderChute/Glider_CoreStreet/FX/NS_Glider_CoreStreet
dependencies {
 usedDataInterfaces [
  &NiagaraDataInterfaceArrayFloat3
  &NiagaraDataInterfaceColorCurve
  &NiagaraDataInterfaceCurve
  &NiagaraDataInterfaceVectorCurve
  &NiagaraDataInterfaceVector2DCurve
 ]
 usedDynamicInputs [&Abs_Float]
 usedModules [
  &SystemState
  &Timeline
  &self:SystemSpawnScript.SystemScriptSource.SystemScriptGraph.NiagaraNodeAssignment_0.SetVariables_#1
 ]
}
emitters {
 EmbersGPUHigh {
  simTarget=GPUComputeSim
  emitterUpdateScript [
   &EmitterState()
   &SpawnRate(
    Spawn Probability=&FloatFromCurve(
     FloatCurve=DataInterface() {propertyValues {Curve {keys [<at 0 0 RCIM_Constant>, <at 0.1 1>, {arriveTangent=-2.5249, leaveTangent=-2.5249, time=0.5, value=0}]}}}
     CurveIndex=$System.Timeline.Time
     Scale Curve=&FN_FrontEndFloatModifier(Gameplay=$User.ForwardDot, Front End=1)
    )
    SpawnRate=444
    Use Spawn Probability=true
   )
  ]
  particleSpawnScript [
   &FN_SetSpawnGroup()
   /SetVariables(Particles.PerParticleRandom:NiagaraFloat=&RandomRangeFloat(Minimum=0, Maximum=1)) {isSetParameters=true}
   &InitializeParticle(
    Color Mode=::NewEnumerator0
    Lifetime=&Lerp_Float(A=0.999, B=0.666, Alpha=$Particles.PerParticleRandom)
    Position Offset=&SelectVectorFromArray(
     Direct Array Index=&SpawnGroup()
     Vector Selection Array=DataInterface() {propertyValues {FloatData [<70 54 28>, <70 -54 28>, <77 67 15.5>, <77 -67 15.5>]}}
    )
    Sprite Size=&Multiply_Vector2D(
     A=<0.5 1>
     B=&MakeVector2D(
      X=&FloatFromCurve(
       FloatCurve=DataInterface() {propertyValues {Curve {keys [<at 0 1>,<at 0.8 2 RCIM_Constant>,<at 0.95 3 RCIM_Constant>]}}}
       CurveIndex=$Particles.PerParticleRandom
      )
      Y=&FloatFromCurve(
       FloatCurve=DataInterface() {propertyValues {Curve {keys [<at 0 1>,<at 0.8 2 RCIM_Constant>,<at 0.95 3>]}}}
       CurveIndex=$Particles.PerParticleRandom
      )
     )
    )
    Sprite Size Mode=::NewEnumerator4
    UsePositionOffset=true
    Write Material Random=true
   )
   &AddVelocity(
    Cone Angle=33
    Module.UseConeFalloff=false
    Module.UseVelDistribution=false
    Override Seed=false
    Velocity Mode=::NewEnumerator2
    Velocity Speed=&Lerp_Float(A=444, B=1111, Alpha=$Particles.PerParticleRandom)
   )
   &ShapeLocation(MaskSpawnGroup=false, Override Seed=false, Sphere Radius=5, UseOwnerScale=true)
  ]
  particleUpdateScript [
   &ParticleState()
   &AccelerationForce()
   &Drag(
    Drag=&FloatFromCurve(
     FloatCurve=DataInterface() {propertyValues {Curve {keys [<at 0 0.5>,<at 0.5 0.5>,<at 1 1>]}}}
     Scale Curve=&RandomRangeFloat(Minimum=6, Maximum=22)
    )
    Use Linear Drag=true
    Use Rotational Drag=false
   )
   /Niagara/Modules/Update/Forces/CurlNoiseForce(
    Noise Frequency=0.04
    Noise Strength=&FloatFromCurve(
     FloatCurve=DataInterface() {propertyValues {Curve {keys [{tangentMode=RCTM_User, time=0, value=0}, {tangentMode=RCTM_User, time=1, value=1}]}}}
     Scale Curve=&Lerp_Float(A=4444, B=444, Alpha=$Particles.PerParticleRandom)
    )
    Pan Noise=true
    Pan Noise Field=<0 0 1>
   )
   &SolveForcesAndVelocity(Clamp Velocity=false, Limit Acceleration=false)
   &ScaleColor(
    Scale RGB=&MakeVectorFromLinearColorRGB(Color=&ColorFromCurve(ColorCurve=DataInterface() {propertyValues {
        AlphaCurve {keys [<at 0 1 RCIM_Cubic>]}
        BlueCurve {keys [<at 0 0 RCIM_Cubic>,<at 1 0 RCIM_Cubic>]}
        GreenCurve {keys [<at 0 9.1667 RCIM_Cubic>,<at 1 0 RCIM_Cubic>]}
        RedCurve {keys [<at 0 50 RCIM_Cubic>,<at 1 1 RCIM_Cubic>]}
       }}))
    ScaleA=true
    ScaleRGB=true
   )
   &ScaleSpriteSizeBySpeed(
    Max Scale Factor=<1.666 13>
    Min Scale Factor=<1.333 1.333>
    Sample Scale Factor By Curve=true
    Scale Factor Curve=DataInterface() {propertyValues {Curve {keys [<at 0 0>, {arriveTangent=1.4989, interpMode=RCIM_Cubic, tangentMode=RCTM_Break, time=1, value=1}]}}}
    Velocity Threshold=999
   )
   &ScaleSpriteSize(
    Uniform Curve Scale=&MakeFloatFromVector(Vector=$Engine.Owner.Scale, Channel=::NewEnumerator0)
    Uniform Curve Sprite Scale=DataInterface() {propertyValues {Curve {keys [
        {tangentMode=RCTM_User, time=0, value=1}
        {arriveTangent=-0.0405, interpMode=RCIM_Cubic, leaveTangent=-0.0405, tangentMode=RCTM_User, time=0.7009, value=1}
        {arriveTangent=-8.0089, interpMode=RCIM_Cubic, leaveTangent=-8.0089, tangentMode=RCTM_User, time=1, value=0}
       ]}}}
   )
  ]
  rendererClasses [&NiagaraSpriteRendererProperties]
  renderers [{rendererClass=&NiagaraSpriteRendererProperties}]
 }
 ExhaustCone {
  emitterUpdateScript [
   &EmitterState(
    Inactive Response=::NewEnumerator1
    Life Cycle Mode=::NewEnumerator1
    Loop Behavior=::NewEnumerator1
    Loop Duration Mode=::NewEnumerator1
    UseLoopDelay=false
   )
   &SpawnBurst_Instantaneous(Spawn Count=4, Use Spawn Probability=false, UseLoopCountLimit=false)
  ]
  particleSpawnScript [
   &FN_SetSpawnGroup()
   &InitializeParticle(
    Color=<rgba 6 0.4 0 1>
    Mesh Scale=&Multiply_VectorByFloat(Vector=$Engine.Owner.Scale, Float=0.666)
    Mesh Scale Mode=::NewEnumerator4
    Position Offset=&SelectVectorFromArray(
     Direct Array Index=&SpawnGroup()
     Vector Selection Array=DataInterface() {propertyValues {FloatData [<55 54 28>, <55 -54 28>, <60 67 16>, <60 -67 16>]}}
    )
    UsePositionOffset=true
    Write Material Random=true
   )
   &InitialMeshOrientation(
    Mesh Axis To Orient=<0 0 1>
    Mesh Orientation Mode=::NewEnumerator2
    Orientation Coordinate Space=::Local
    Rotation=&RandomRangeVector(Minimum=<0 0 0>, Maximum=<0 0 1>)
    Use Rotation Vector=true
   )
  ]
  particleUpdateScript [
   &ParticleState(Kill Particles When Lifetime Has Elapsed=false)
   &DynamicMaterialParameters(
    Index 0 Param 1=&RandomRangeFloat(Minimum=0.25, Maximum=0.6)
    Index 0 Param 2=&RandomRangeFloat(Minimum=1, Maximum=2, Evaluation Type=::NewEnumerator1)
    Index 0 Param 3=1
    Index 0 Param 4=1
    Module.Param0WriteEnabled=true
    Module.Param1WriteEnabled=true
    Module.Param2WriteEnabled=true
    Module.Param3WriteEnabled=false
   )
   &ScaleMeshSize(Scale Factor=&MakeVector(
     X=1
     Y=1
     Z=&FloatFromCurve(
      FloatCurve=DataInterface() {propertyValues {Curve {keys [
          {interpMode=RCIM_Cubic, tangentMode=RCTM_User, time=0, value=0.55}
          {interpMode=RCIM_Cubic, tangentMode=RCTM_User, time=0.5, value=0.65}
          {tangentMode=RCTM_User, time=1, value=0.55}
         ]}}}
      CurveIndex=&Frac_Float(Float=&Add_Float(
        A=&Multiply_Float(A=$Engine.Time, B=3)
        B=&RandomRangeFloat(Minimum=0, Maximum=1)
       ))
     )
    ))
   &Color(
    Color=&ColorFromCurve(
     ColorCurve=DataInterface() {propertyValues {
       AlphaCurve {keys [<at 0 0 RCIM_Cubic>,<at 1 1 RCIM_Cubic>]}
       BlueCurve {keys [<at 0 0 RCIM_Cubic>,<at 0.333 0>,<at 0.666 0>,<at 1 0 RCIM_Cubic>]}
       GreenCurve {keys [<at 0 0 RCIM_Cubic>,<at 0.333 0>,<at 0.666 0.0333>,<at 1 0.3 RCIM_Cubic>]}
       RedCurve {keys [<at 0 0 RCIM_Cubic>,<at 0.333 0.666>,<at 0.666 2>,<at 1 3 RCIM_Cubic>]}
      }}
     CurveIndex=$System.Timeline.Time
    )
    Scale Color=&VectorFromFloat(Value=&FN_FrontEndFloatModifier(Gameplay=1, Front End=3))
   )
  ]
  rendererClasses [&NiagaraMeshRendererProperties]
  renderers [{rendererClass=&NiagaraMeshRendererProperties, rendererIndex=0}]
 }
 FN_Debug {
  bEnabled=false
  emitterUpdateScript [
   &EmitterState(Life Cycle Mode=::NewEnumerator1, Loop Behavior=::NewEnumerator1, Loop Duration Mode=::NewEnumerator1, UseLoopDelay=false)
   &SpawnBurst_Instantaneous(Spawn Count=1, Use Spawn Probability=false)
  ]
  particleSpawnScript [/Niagara/Modules/Spawn/Initialization/InitializeParticle(
    Lifetime=5
    Module.Write Color=true
    Module.Write Lifetime=true
    Module.Write Mass=true
    Module.Write Material Random=true
    Module.Write Position=false
    Module.Write Scale=false
    Module.Write SpriteRotation=false
    Module.Write SpriteSize=true
    Position=&SimulationPosition()
    Sprite Size=<100 100>
   )]
  particleUpdateScript [
   &ParticleState(Kill Particles When Lifetime Has Elapsed=false)
   &FN_Timer()
   /SetVariables(Particles.Debug:NiagaraFloat=$System.Timeline.Time) {isSetParameters=true}
   &DynamicMaterialParameters(
    Index 0 Param 1=$Particles.Debug
    Module.Param0WriteEnabled=true
    Module.Param1WriteEnabled=true
    Module.Param2WriteEnabled=true
    Module.Param3WriteEnabled=true
   )
   &CameraOffset(Camera Offset Amount=111)
  ]
  rendererClasses [&NiagaraSpriteRendererProperties]
  renderers [{rendererClass=&NiagaraSpriteRendererProperties}]
 }
 GlowCard {
  emitterUpdateScript [
   &EmitterState(
    Inactive Response=::NewEnumerator1
    Life Cycle Mode=::NewEnumerator1
    Loop Behavior=::NewEnumerator1
    Loop Duration Mode=::NewEnumerator1
    UseLoopDelay=false
   )
   &SpawnBurst_Instantaneous(Spawn Count=4, Use Spawn Probability=false, UseLoopCountLimit=false)
  ]
  particleSpawnScript [
   &FN_SetSpawnGroup()
   /Niagara/Modules/Spawn/Initialization/V2/InitializeParticle(
    Color=<rgba 333 33.3 0 1>
    Position Offset=&SelectVectorFromArray(
     Direct Array Index=&SpawnGroup()
     Vector Selection Array=DataInterface() {propertyValues {FloatData [<70 54 28>, <70 -54 28>, <76 67 16>, <76 -67 16>]}}
    )
    Sprite Size Mode=::NewEnumerator3
    Uniform Sprite Size=&Multiply_Float(A=11, B=&MakeFloatFromVector(Vector=$Engine.Owner.Scale, Channel=::NewEnumerator0))
    UsePositionOffset=true
    Write Material Random=true
   )
   &SpriteFacingAndAlignment(WriteAlignment=false, WriteFacing=true)
  ]
  particleUpdateScript [
   &ParticleState(Kill Particles When Lifetime Has Elapsed=false)
   &ScaleColor(Scale Alpha=$System.Timeline.Time, ScaleA=true, ScaleRGB=false)
  ]
  rendererClasses [&NiagaraSpriteRendererProperties]
  renderers [{rendererClass=&NiagaraSpriteRendererProperties}]
 }
 GlowSprites {
  emitterUpdateScript [
   &EmitterState(
    Inactive Response=::NewEnumerator1
    Life Cycle Mode=::NewEnumerator1
    Loop Behavior=::NewEnumerator1
    Loop Duration Mode=::NewEnumerator1
    UseLoopDelay=false
   )
   &SpawnBurst_Instantaneous(Spawn Count=4, Use Spawn Probability=false, UseLoopCountLimit=false)
  ]
  particleSpawnScript [
   &FN_SetSpawnGroup()
   &InitializeParticle(
    Color=<rgba 0.5 0.0575 0.05 1>
    Position Offset=&SelectVectorFromArray(
     Direct Array Index=&SpawnGroup()
     Vector Selection Array=DataInterface() {propertyValues {FloatData [<77 54 28.5>, <77 -54 28.5>, <84 67 16.5>, <84 -67 16.5>]}}
    )
    Sprite Size Mode=::NewEnumerator3
    Uniform Sprite Size=&Multiply_Float(A=44, B=&MakeFloatFromVector(Vector=$Engine.Owner.Scale, Channel=::NewEnumerator0))
    UsePositionOffset=true
    Write Material Random=true
   )
   &CameraOffset(Camera Offset Amount=11)
  ]
  particleUpdateScript [
   &ParticleState(Kill Particles When Lifetime Has Elapsed=false)
   &ScaleColor(Scale Alpha=&Lerp_Float(A=0, B=0.333, Alpha=$System.Timeline.Time), ScaleA=true, ScaleRGB=false)
  ]
  rendererClasses [&NiagaraSpriteRendererProperties]
  renderers [{rendererClass=&NiagaraSpriteRendererProperties}]
 }
 Jets {
  emitterUpdateScript [
   &EmitterState()
   &SpawnRate(
    Spawn Probability=&FN_FrontEndFloatModifier(Gameplay=&Power(a=$User.ForwardDot, b=0.45), Front End=1)
    SpawnRate=400
    Use Spawn Probability=true
   )
  ]
  particleSpawnScript [
   &InitializeParticle(
    Lifetime=0.06
    Position Offset=&SelectVectorFromArray(
     Array Sampling Mode=::NewEnumerator0
     Direct Array Index=?
     Vector Selection Array=DataInterface() {propertyValues {FloatData [<70 54 28>, <70 -54 28>, <77 67 15.5>, <77 -67 15.5>]}}
     Random Seed=0
    )
    Sprite Size=&Multiply_Vector2D_ByFloat(Float=&MakeFloatFromVector(Vector=$Engine.Owner.Scale, Channel=::NewEnumerator0), Vector2D=<15 5>)
    Sprite Size Mode=::NewEnumerator4
    UsePositionOffset=true
    Write Material Random=true
   )
   &AddVelocity(Velocity=<333 0 0>)
   &Color(Color=&MakeLinearColorFromVectorAndFloat("Vector (RGB)"=<5 2 1>, "Float (Alpha)"=&ClampFloat(Float=0.1)))
  ]
  particleUpdateScript [
   &ParticleState()
   &SolveForcesAndVelocity(Clamp Velocity=false, Limit Acceleration=false)
   &ScaleColor(
    Scale Alpha=&FloatFromCurve(
     FloatCurve=DataInterface() {propertyValues {Curve {keys [<at 0 0 RCIM_Cubic>,<at 0.25 1 RCIM_Cubic>,<at 1 0 RCIM_Cubic>]}}}
     Scale Curve=&FN_FrontEndFloatModifier(Gameplay=&Power(a=$User.ForwardDot, b=0.2), Front End=1)
    )
    Scale RGB=&VectorFromCurve(VectorCurve=DataInterface() {propertyValues {XCurve {keys [<at 0 1 RCIM_Cubic>,<at 1 50>]}, YCurve {keys [<at 0 1 RCIM_Cubic>,<at 1 15>]}, ZCurve {keys [<at 0 1 RCIM_Cubic>,<at 1 5>]}}})
    ScaleA=true
    ScaleRGB=true
   )
   &ScaleSpriteSize(
    Non-Uniform Curve Sprite Scale=DataInterface() {propertyValues {
      XCurve {keys [{tangentMode=RCTM_User, time=0, value=1}, {tangentMode=RCTM_User, time=1, value=0.1}]}
      YCurve {keys [{tangentMode=RCTM_User, time=0, value=1}, {tangentMode=RCTM_User, time=1, value=10}]}
     }}
    Scale Sprite Size Mode=::NewEnumerator3
   )
  ]
  rendererClasses [&NiagaraSpriteRendererProperties]
  renderers [{rendererClass=&NiagaraSpriteRendererProperties}]
 }
 Lights {
  emitterUpdateScript [
   &EmitterState(
    Inactive Response=::NewEnumerator1
    Life Cycle Mode=::NewEnumerator1
    Loop Behavior=::NewEnumerator1
    Loop Duration Mode=::NewEnumerator1
    UseLoopDelay=false
   )
   &SpawnBurst_Instantaneous(Spawn Count=2, Use Spawn Probability=false, UseLoopCountLimit=false)
  ]
  particleSpawnScript [
   &FN_SetSpawnGroup()
   &InitializeParticle(
    Color=<rgba 3 0.795 0.3 1>
    Position Offset=&SelectVectorFromArray(
     Direct Array Index=&SpawnGroup()
     Vector Selection Array=DataInterface() {propertyValues {FloatData [<111 60 25>, <111 -60 25>]}}
    )
    UsePositionOffset=true
    Write Material Random=true
   )
   /SetVariables(Particles.LightRadius=&Multiply_Float(A=&MakeFloatFromVector(Vector=$Engine.Owner.Scale, Channel=::NewEnumerator0), B=166)) {isSetParameters=true}
  ]
  particleUpdateScript [
   &ParticleState(Kill Particles When Lifetime Has Elapsed=false)
   &ScaleColor(Scale Alpha=$System.Timeline.Time, ScaleA=true, ScaleRGB=false)
   /SetVariables(Particles.LightEnabled:NiagaraBool=&SetBoolByFloatComparison(A=$System.Timeline.Time, B=0)) {isSetParameters=true}
  ]
  rendererClasses [&NiagaraLightRendererProperties]
  renderers [{rendererClass=&NiagaraLightRendererProperties, rendererIndex=0}]
 }
 Muzzle_Core_Elements {
  emitterUpdateScript [
   &EmitterState()
   &SpawnRate(Spawn Probability=&Lerp_Float(A=0.3, B=0, Alpha=$User.ForwardDot), SpawnRate=20, Use Spawn Probability=true)
   &SpawnBurst_Instantaneous(Use Spawn Probability=false, UseLoopCountLimit=false) {enabled=false}
  ]
  particleSpawnScript [
   /SetVariables(Particles.VisibilityTag=&SpawnGroup()) {isSetParameters=true}
   &PartitionParticles(Number Of Partitions=&MaskIntBySpawnGroup(Masked Int=0, Passthrough Int=6, Spawn Group Mask=0))
   &InitializeParticle(
    Color=<rgba 2 1 0.1 1>
    Lifetime=&RandomRangeFloat(Minimum=0.075, Maximum=0.1)
    Mesh Scale=<0.8 0.5 0.5>
    Mesh Scale Mode=::NewEnumerator4
    Position Offset=&SelectVectorFromArray(
     Array Sampling Mode=::NewEnumerator0
     Direct Array Index=?
     Vector Selection Array=DataInterface() {propertyValues {FloatData [<70 54 28>, <70 -54 28>, <77 67 15.5>, <77 -67 15.5>]}}
     Random Seed=0
    )
    Sprite Rotation Mode=::NewEnumerator2
    Sprite Size=<50 50>
    Sprite Size Mode=::NewEnumerator4
    UsePositionOffset=true
    Write Material Random=true
   )
   &InitialMeshOrientation(
    Mesh Orientation Mode=::NewEnumerator2
    Orientation Vector=&MaskVectorBySpawnGroup(
     Masked Vector=<1 0 0>
     Passthrough Vector=&self:Rotation(
      Default:Vector3f=<0 1 0>
      float:NiagaraFloat=&Divide_Float(A=&ReturnNormalizedExecIndex(Normalized Index Includes One=false), B=1)
     )
    )
    Rotation=&RandomRangeVector(Minimum=<0 0 0>, Maximum=<1 0 0>)
    Use Rotation Vector=true
   )
   &CameraOffset(Camera Offset Amount=1)
   &SpriteFacingAndAlignment(
    Sprite Alignment=&MaskVectorBySpawnGroup(Masked Vector=<0 1 0>, Passthrough Vector=<1 0 0>)
    Sprite Facing=&MaskVectorBySpawnGroup(
     Masked Vector=<1 0 0>
     Passthrough Vector=&Lerp_Vector(A=<0 1 0>, B=<0 -1 0>, Alpha=&MakeFloatFromInt(Int=$Particles.PartitionParticles.Partition))
    )
    WriteAlignment=true
    WriteFacing=true
   )
   /SetVariables(Particles.SpriteCameraOffset=5) {isSetParameters=true}
  ]
  particleUpdateScript [
   &ParticleState()
   &ScaleSpriteSize(
    Scale Factor=&Vector2DFromFloat(Value=&FloatFromCurve(
      FloatCurve=DataInterface() {propertyValues {Curve {keys [{tangentMode=RCTM_User, time=0, value=0}, {tangentMode=RCTM_User, time=1, value=1}]}}}
      Scale Curve=0.6
     ))
    Scale Sprite Size Mode=::NewEnumerator2
   )
   &ScaleMeshSize(Scale Factor=&VectorFromCurve(VectorCurve=DataInterface() {propertyValues {XCurve {keys [<at 0 0>,<at 1 1>]}, YCurve {keys [<at 0 0.25>,<at 1 1.25>]}, ZCurve {keys [<at 0 0.25>,<at 1 1.25>]}}}))
   &ScaleColor(
    Scale Alpha=&RandomRangeFloat(Minimum=0.5, Maximum=1)
    Scale RGB=&MaskVectorBySpawnGroup(
     Masked Vector=&VectorFromFloat(Value=&FloatFromCurve(
       FloatCurve=DataInterface() {propertyValues {Curve {keys [{arriveTangent=-42.9916, interpMode=RCIM_Cubic, leaveTangent=-42.9916, tangentMode=RCTM_User, time=0, value=5}, <at 0.15 1 RCIM_Cubic>]}}}
       Scale Curve=0.5
      ))
     Passthrough Vector=<1 1 1>
    )
    ScaleA=true
    ScaleRGB=true
   )
   &DynamicMaterialParameters(
    Index 0 Param 1=&FloatFromCurve(
     FloatCurve=DataInterface() {propertyValues {Curve {keys [
         {arriveTangent=2.9073, interpMode=RCIM_Cubic, leaveTangent=2.9073, leaveTangentWeight=0.3077, tangentMode=RCTM_User, time=0.35, value=0}
         {arriveTangentWeight=0.8354, interpMode=RCIM_Cubic, tangentMode=RCTM_User, time=1, value=1}
        ]}}}
     Scale Curve=0.75
    )
    Index 0 Param 2=&FloatFromCurve(FloatCurve=DataInterface() {propertyValues {Curve {keys [
         {interpMode=RCIM_Cubic, leaveTangentWeight=0.3627, tangentMode=RCTM_User, time=0.4, value=1}
         {arriveTangent=-1.799, arriveTangentWeight=1.0345, interpMode=RCIM_Cubic, leaveTangent=-1.799, tangentMode=RCTM_User, time=1, value=0}
        ]}}})
    Index 0 Param 3=&MakeCustomFloatFromBool(Bool=&RandomBool(), True Float=1, False Float=0)
    Module.Index0Param0Write=true
    Module.Index0Param1Write=true
    Module.Index0Param2Write=true
    Module.Index0Param3Write=false
   )
   /SetVariables(Particles.SpritePOsition:NiagaraPosition=&AddVectorToPosition(Position=$Particles.Position, Vector=<2 0 0>)) {isSetParameters=true}
   /SetVariables(
    Particles.InnerGlowColor:LinearColor=&Multiply_LinearColorByFloat(
     Linear Color=<rgba 333 33.3 0 1>
     Float=&FloatFromCurve(
      FloatCurve=DataInterface() {propertyValues {Curve {keys [
          {interpMode=RCIM_Cubic, tangentMode=RCTM_User, time=0, value=0}
          {interpMode=RCIM_Cubic, tangentMode=RCTM_User, time=0.5, value=1}
          {tangentMode=RCTM_User, time=1, value=0}
         ]}}}
      Scale Curve=0.1
     )
    )
    Particles.InnerGlowPosition:NiagaraPosition=&AddVectorToPosition(Position=$Particles.SpritePOsition, Vector=<0 0 0>)
    Particles.InnerGlowSpriteSize:Vector2f=&Multiply_Vector2D_ByFloat(Float=0.5, Vector2D=<20 20>)
   ) {isSetParameters=true}
   /SetVariables(
    Particles.AirGlowCameraOffset=10
    Particles.AirGlowColor:LinearColor=&Multiply_LinearColorByFloat(
     Linear Color=<rgba 333 33.3 0 1>
     Float=&FloatFromCurve(
      FloatCurve=DataInterface() {propertyValues {Curve {keys [
          {interpMode=RCIM_Cubic, tangentMode=RCTM_User, time=0, value=0}
          {interpMode=RCIM_Cubic, tangentMode=RCTM_User, time=0.5, value=1}
          {tangentMode=RCTM_User, time=1, value=0}
         ]}}}
      Scale Curve=0.03
     )
    )
    Particles.AirGlowPosition:NiagaraPosition=&AddVectorToPosition(Position=$Particles.SpritePOsition, Vector=<10 0 0>)
    Particles.AirGlowSpriteSize:Vector2f=&Multiply_Vector2D_ByFloat(Float=1.25, Vector2D=<20 20>)
   ) {isSetParameters=true}
   &FN_Light_Attributes(
    Color=<rgba 333 33.3 0 1>
    Create Color Attribute=true
    Create Exponent Attribute=true
    Create Position Attribute=true
    Create Radius Attribute=true
    Enable Particle Light Enable=true
    Exponent=2
    Light Enabled=true
    Position=&AddVectorToPosition(Position=$Particles.SpritePOsition, Vector=<10 0 0>)
    Radius=30
    Scale Color=0.01
   )
  ]
  rendererClasses [
   &NiagaraMeshRendererProperties
   &NiagaraSpriteRendererProperties
   &NiagaraLightRendererProperties
  ]
  renderers [
   {rendererClass=&NiagaraMeshRendererProperties, rendererIndex=0}
   {rendererClass=&NiagaraSpriteRendererProperties, rendererIndex=1}
   {rendererClass=&NiagaraMeshRendererProperties, rendererIndex=2}
   {rendererClass=&NiagaraSpriteRendererProperties, rendererIndex=3}
   {rendererClass=&NiagaraSpriteRendererProperties, rendererIndex=4}
   {rendererClass=&NiagaraSpriteRendererProperties, rendererIndex=5}
   {rendererClass=&NiagaraLightRendererProperties, rendererIndex=6}
  ]
 }
 SmokeLargeLocalSpace {
  emitterUpdateScript [
   &EmitterState(Inactive Response=::NewEnumerator1, Life Cycle Mode=::NewEnumerator1, Loop Delay=0.15, Loop Duration=0, UseLoopDelay=false)
   &SpawnRate(
    Spawn Probability=&FN_FrontEndFloatModifier(Gameplay=&Power(a=$User.ForwardDot, b=0.2), Front End=1)
    SpawnRate=33
    Use Spawn Probability=true
   )
  ]
  particleSpawnScript [
   &FN_SetSpawnGroup(Number of Spawn Groups=2)
   &InitializeParticle(
    Color=&MakeLinearColorFromVectorAndFloat("Vector (RGB)"=<1 1 1>, "Float (Alpha)"=$System.Timeline.Time)
    Lifetime=0.55
    Position Offset=&SelectVectorFromArray(
     Direct Array Index=&SpawnGroup()
     Vector Selection Array=DataInterface() {propertyValues {FloatData [<85 60 25>, <85 -60 25>]}}
    )
    Sprite Size Mode=::NewEnumerator3
    Uniform Sprite Size=&Multiply_Float(A=77, B=&MakeFloatFromVector(Vector=$Engine.Owner.Scale, Channel=::NewEnumerator0))
    UsePositionOffset=true
    Write Material Random=true
   )
  ]
  particleUpdateScript [
   &ParticleState()
   &Color(
    Color=&ColorFromCurve(ColorCurve=DataInterface() {propertyValues {
       AlphaCurve {keys [<at 0 0.2>,<at 0.1 1>,<at 0.75 0.666>,<at 1 0>]}
       BlueCurve {keys [<at 0 0>, {arriveTangent=0.0577, time=0.333, value=0.075}, <at 1 0.05>]}
       GreenCurve {keys [<at 0 0.1833>, {arriveTangent=-0.0961, time=0.333, value=0.075}, <at 1 0.05>]}
       RedCurve {keys [<at 0 1>, {arriveTangent=-1.0957, time=0.333, value=0.075}, <at 1 0.05>]}
      }})
    Scale Alpha=&MakeFloatFromLinearColor(LinearColor=$Particles.Initial.Color, Channel:ENiagaraLinearColor_Channels=::NewEnumerator3)
   )
   &ScaleSpriteSize(
    Uniform Curve Scale=&RandomRangeFloat(Minimum=0.8, Maximum=1.1)
    Uniform Curve Sprite Scale=DataInterface() {propertyValues {Curve {keys [<at 0 0>, {leaveTangent=1.5237, time=0.1, value=0.6}, <at 0.6 1.1>, {arriveTangent=-5.0995, time=1, value=0.5}]}}}
   )
   &DynamicMaterialParameters(
    Index 0 Param 1=&Add_Float(
     A=&Multiply_Float(A=$Engine.Time, B=1)
     B=&RandomRangeFloat(Minimum=0, Maximum=1)
    )
    Index 0 Param 3=&FloatFromCurve(
     FloatCurve=DataInterface() {propertyValues {Curve {keys [<at 0 1>,<at 0.333 0>]}}}
     CurveIndex=$Particles.Age
     Scale Curve=&Multiply_Float(A=0.02, B=$System.Timeline.Time)
    )
    Index 2 Param 1=1
    Index 2 Param 2=&RandomRangeFloat(Minimum=0, Maximum=1)
    Index 2 Param 3=&FloatFromCurve(FloatCurve=DataInterface() {propertyValues {Curve {keys [
         {interpMode=RCIM_Cubic, tangentMode=RCTM_Break, time=0.666, value=0}
         {arriveTangent=10.9325, interpMode=RCIM_Cubic, tangentMode=RCTM_Break, time=1, value=2}
        ]}}})
    Index 2 Param 4=1
    Module.Param0WriteEnabled=true
    Module.Param0WriteEnabled1=true
    Module.Param0WriteEnabled2=true
    Module.Param1WriteEnabled=true
    Module.Param1WriteEnabled1=true
    Module.Param1WriteEnabled2=true
    Module.Param2WriteEnabled=true
    Module.Param2WriteEnabled1=true
    Module.Param2WriteEnabled2=true
    Module.Param3WriteEnabled=true
    Module.Param3WriteEnabled1=true
    Module.Param3WriteEnabled2=true
    Write Parameter Index 1=true
    Write Parameter Index 2=true
   )
   /SetVariables(Particles.Position=&AddVectorToPosition(
     Position=$Particles.Position
     Vector=&Multiply_VectorByFloat(
      Vector=&NormalizeVector(
       Vector To Normalize=&Lerp_Vector(
        A=<1 0 0>
        B=&TransformVector(
         Vector=&NormalizeVector(
          Vector To Normalize=&Multiply_Vector(A=$Engine.Owner.Velocity, B=<-1 -1 -3>)
          Safe Normalize=true
          Safe Fallback Threshold=0
          Safe Fallback Vector=<0 0 0.001>
         )
         Source Space=::World
         Destination Space=::Local
        )
        Alpha=&FloatFromCurve(
         FloatCurve=DataInterface() {propertyValues {Curve {keys [{interpMode=RCIM_Cubic, leaveTangent=2.5306, tangentMode=RCTM_Break, time=0, value=0}, {tangentMode=RCTM_User, time=1, value=1}]}}}
         Scale Curve=$System.ForwardAndRightDot
        )
       )
       Safe Normalize=true
       Safe Fallback Threshold=0
       Safe Fallback Vector=<0 0 0.001>
      )
      Float=9
     )
    )) {isSetParameters=true}
  ]
  rendererClasses [&NiagaraSpriteRendererProperties]
  renderers [
   {rendererClass=&NiagaraSpriteRendererProperties}
   {rendererClass=&NiagaraSpriteRendererProperties, rendererIndex=1}
  ]
 }
 SmokeSmallLocalSpace {
  emitterUpdateScript [
   &EmitterState(Life Cycle Mode=::NewEnumerator1, Loop Delay=0.15, Loop Duration=0, UseLoopDelay=false)
   &SpawnPerFrame(
    Spawn Count=4
    Spawn Probability:NiagaraFloat=&FN_FrontEndFloatModifier(Gameplay=&Power(a=$User.ForwardDot, b=0.2), Front End=1)
    Use Spawn Probability=true
   )
  ]
  particleSpawnScript [
   &FN_SetSpawnGroup()
   &InitializeParticle(
    Color=<rgba 1 0.15 0 1>
    Lifetime=0.2
    Position Offset=&SelectVectorFromArray(
     Direct Array Index=&SpawnGroup()
     Vector Selection Array=DataInterface() {propertyValues {FloatData [<70 54 28>, <70 -54 28>, <77 67 15.5>, <77 -67 15.5>]}}
    )
    Sprite Size Mode=::NewEnumerator3
    Uniform Sprite Size=&Multiply_Float(A=16, B=&MakeFloatFromVector(Vector=$Engine.Owner.Scale, Channel=::NewEnumerator0))
    UsePositionOffset=true
    Write Material Random=true
   )
   &CameraOffset()
  ]
  particleUpdateScript [
   &ParticleState()
   &ScaleSpriteSize(
    Uniform Curve Scale=&RandomRangeFloat(Minimum=0.8, Maximum=1)
    Uniform Curve Sprite Scale=DataInterface() {propertyValues {Curve {keys [<at 0 1>, {arriveTangent=1.5721, time=1, value=2}]}}}
   )
   &Color(
    Color=&ColorFromCurve(ColorCurve=DataInterface() {propertyValues {
       AlphaCurve {keys [<at 0 1>, {leaveTangent=-0.3445, time=0.75, value=0.666}, <at 1 0>]}
       BlueCurve {keys [
         {interpMode=RCIM_Cubic, tangentMode=RCTM_Break, time=0.666, value=0}
         {arriveTangent=0.0941, interpMode=RCIM_Cubic, tangentMode=RCTM_Break, time=1, value=0.08}
        ]}
       GreenCurve {keys [
         {interpMode=RCIM_Cubic, tangentMode=RCTM_Break, time=0.666, value=0.1833}
         {arriveTangent=-0.1174, interpMode=RCIM_Cubic, tangentMode=RCTM_Break, time=1, value=0.086}
        ]}
       RedCurve {keys [{interpMode=RCIM_Cubic, tangentMode=RCTM_Break, time=0.666, value=1}, {arriveTangent=-1.0957, time=1, value=0.2}]}
      }})
    Scale Alpha=&MakeFloatFromLinearColor(LinearColor=$Particles.Initial.Color, Channel:ENiagaraLinearColor_Channels=::NewEnumerator3)
   )
   &DynamicMaterialParameters(
    Index 0 Param 1=&Add_Float(
     A=&Multiply_Float(A=$Engine.Time, B=1)
     B=&RandomRangeFloat(Minimum=0, Maximum=1)
    )
    Index 0 Param 3=&FloatFromCurve(
     FloatCurve=DataInterface() {propertyValues {Curve {keys [<at 0 1>,<at 0.333 0>]}}}
     Scale Curve=&RandomRangeFloat(Minimum=0.111, Maximum=0.333)
    )
    Index 1 Param 1=9
    Index 1 Param 4=0.5
    Module.Param0WriteEnabled=true
    Module.Param0WriteEnabled1=true
    Module.Param0WriteEnabled2=true
    Module.Param1WriteEnabled=true
    Module.Param1WriteEnabled1=true
    Module.Param1WriteEnabled2=true
    Module.Param2WriteEnabled=true
    Module.Param2WriteEnabled1=true
    Module.Param2WriteEnabled2=true
    Module.Param3WriteEnabled=true
    Module.Param3WriteEnabled1=true
    Module.Param3WriteEnabled2=true
    Write Parameter Index 1=true
    Write Parameter Index 2=true
   )
   &ScaleColor(
    Color Value To Scale=$Particles.Color
    Scale Alpha=&Power(a=$System.Timeline.Time, b=2)
    Scale RGB=&VectorFromFloat(Value=$System.Timeline.Time)
    ScaleA=true
    ScaleRGB=true
   )
   /SetVariables(Particles.Position=&AddVectorToPosition(
     Position=$Particles.Position
     Vector=&Multiply_VectorByFloat(
      Vector=&NormalizeVector(
       Vector To Normalize=&Lerp_Vector(
        A=<1 0 0>
        B=&TransformVector(
         Vector=&NormalizeVector(
          Vector To Normalize=&Multiply_Vector(A=$Engine.Owner.Velocity, B=<-1 -1 -3>)
          Safe Normalize=true
          Safe Fallback Threshold=0
          Safe Fallback Vector=<0 0 0.001>
         )
         Source Space=::World
         Destination Space=::Local
        )
        Alpha=&FloatFromCurve(
         FloatCurve=DataInterface() {propertyValues {Curve {keys [{interpMode=RCIM_Cubic, tangentMode=RCTM_Break, time=0, value=0}, {tangentMode=RCTM_User, time=1, value=0.666}]}}}
         Scale Curve=&FN_FrontEndFloatModifier(Gameplay=$System.ForwardAndRightDot, Front End=0)
        )
       )
       Safe Normalize=true
       Safe Fallback Threshold=0
       Safe Fallback Vector=<0 0 0.001>
      )
      Float=&RandomRangeFloat(Minimum=5, Maximum=7)
     )
    )) {isSetParameters=true}
  ]
  rendererClasses [&NiagaraSpriteRendererProperties]
  renderers [{rendererClass=&NiagaraSpriteRendererProperties}]
 }
}
summary {
 systemName=NS_Glider_CoreStreet
 emitters [
  {emitterName=GlowCard, rendererClasses [&NiagaraSpriteRendererProperties]}
  {bEnabled=false, emitterName=FN_Debug, rendererClasses [&NiagaraSpriteRendererProperties]}
  {emitterName=SmokeLargeLocalSpace, rendererClasses [&NiagaraSpriteRendererProperties]}
  {emitterName=SmokeSmallLocalSpace, rendererClasses [&NiagaraSpriteRendererProperties]}
  {emitterName=ExhaustCone, rendererClasses [&NiagaraMeshRendererProperties]}
  {emitterName=Jets, rendererClasses [&NiagaraSpriteRendererProperties]}
  {emitterName=EmbersGPUHigh, simTarget=GPUComputeSim, rendererClasses [&NiagaraSpriteRendererProperties]}
  {emitterName=GlowSprites, rendererClasses [&NiagaraSpriteRendererProperties]}
  {emitterName=Lights, rendererClasses [&NiagaraLightRendererProperties]}
  {
   emitterName=Muzzle_Core_Elements
   rendererClasses [
    &NiagaraMeshRendererProperties
    &NiagaraSpriteRendererProperties
    &NiagaraLightRendererProperties
   ]
  }
 ]
 userVariables [
  User.AmmoType {type=NiagaraInt32, defaultValue=1}
  User.bIsFrontEnd {type=NiagaraBool}
  User.bIsFrontEndPreview {type=NiagaraBool}
  User.ForwardDot {type=NiagaraFloat, defaultValue=1, description=""}
  User.MuzzleCoreCapScale {type=NiagaraFloat, defaultValue=1, description=""}
  User.MuzzleCoreScale {type=Vector3f, description="", defaultValue {struct=Vector3f, value=<1 1 1>}}
  User.RibbonIDOffset {type=NiagaraInt32}
  User.RightDot {type=NiagaraFloat, defaultValue=0, description=""}
  User.SkeletalMesh {
   type=NiagaraDataInterfaceSkeletalMesh
   defaultValue {
    struct=NiagaraExt_VariableValue_DataInterface
    value {dataInterface=&self:NiagaraDataInterfaceSkeletalMesh_0}
   }
  }
 ]
}
